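# Match history row button for the tournament leaderboard, showing a session's number, placement, eliminations and points in a four-column grid.
# PreConstruct registers and styles the text blocks; OnListItemObjectSet fills them from the session team entry view model's tracked-stat map.
# RefreshPointsLayout hides the Points block and sets its column fill from GetShowMatchHistoryPoints.
# SetIndex updates the label and background color, dimming sessions that do not contribute to the score.
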WidgetBlueprint /Compete/TournamentLeaderboard/ModalMessages/MatchHistory/WBP_MatchHistoryListButton
graphs [
 EventGraph {nodes [
   n1 = Event PreConstruct()
   > Parent: PreConstruct(IsDesignTime=@n1.IsDesignTime)
   > RegisterTexts()
   Event On Pressed() {is=title K2Node_Event}
   > Parent: On Pressed() {is=title K2Node_CallParentFunction}
   > n2 = Branch(Condition=GetSelected!(self:Common Button Base Object Reference) -> (ReturnValue:bool))
   PlaySoundPressed(execute=@n2.else, self:WBP Compete Tile Button Base Object Reference)
   n3 = Event OnListItemObjectSet()
   > n4 = Cast To FortTournamentLeaderboardSessionTeamEntryVM!(execute, Object=@n3.ListItemObject) -> (then, CastFailed, AsTournament Leaderboard Session Team Entry VM:Tournament Leaderboard Session Team Entry VM Object Reference) {is=title K2Node_DynamicCast}
   > n5 = Set LeaderboardSessionTeamEntryVM(LeaderboardSessionTeamEntryVM:Tournament Leaderboard Session Team Entry VM Object Reference=@n4) -> (Output_Get:Tournament Leaderboard Session Team Entry VM Object Reference)
   > SetText!(
    execute
    self:WBP UIKit Block Text Object Reference=Get Index() -> (Index:WBP UIKit Block Text Object Reference)
    Text:Text="To Text (Integer)"(Value=Get SessionNumber!(self:Tournament Leaderboard Session Team Entry VM Object Reference=@n5) -> (SessionNumber:int=0))
   ) -> (then)
   > Set Tracked Stat Times Achieved!(
    execute
    self
    Item:WBP UIKit Block Text Object Reference=Get Placement() -> (Placement:WBP UIKit Block Text Object Reference)
    StatString:"String (by ref)"=PLACEMENT_STAT_INDEX
   ) -> (then) {is=title K2Node_CallFunction}
   > Set Tracked Stat Times Achieved(Item=Get Eliminations() -> (Eliminations:WBP UIKit Block Text Object Reference), StatString=TEAM_ELIMS_STAT_INDEX)
   > Set Tracked Stat Times Achieved(Item=Get Points() -> (Points:WBP UIKit Block Text Object Reference), StatString=SCORE_STAT_INDEX)
   > RefreshPointsLayout()
  ]}
 MakeTextDimmed {nodes [
   n1 = MakeTextDimmed!() -> (then, bDim:bool=~) {is=title K2Node_FunctionEntry}
   > Set RenderOpacity(RenderOpacity:float=Select(Option 0:float=1, Option 1:float=0.5, Index=@n1) -> (ReturnValue:float)) -> (Output_Get:float)
  ]}
 RefreshPointsLayout {nodes [
   RefreshPointsLayout() {is=title K2Node_FunctionEntry}
   > SetBlockVisibility!(execute, self:WBP UIKit Block Base Object Reference=Get Points(), Visible:bool=false) -> (then)
   > SetColumnFill!(
    execute
    self:Grid Panel Object Reference=Get Main() -> (Main:Grid Panel Object Reference)
    ColumnIndex:int=3
    Coefficient:float="To Float (Boolean)"(InBool=GetShowMatchHistoryPoints!() -> (ReturnValue:bool))
   ) -> (then)
  ]}
 RegisterTexts {nodes [
   RegisterTexts() {is=title K2Node_FunctionEntry}
   > n1 = Sequence!(execute) -> (then_0, then_1, then_2:Exec, then_3:Exec)
   > RegisterModularBlock!(execute, self:UIKit Modular Button Object Reference, Block:UIKit Block Object Reference=Get Index()) -> (then)
   > RegisterModularBlock(Block=Get Placement())
   > RegisterModularBlock(Block=Get Eliminations())
   > RegisterModularBlock(Block=Get Points())
   SetFont!(
    execute=@n1.then_1
    self:WBP UIKit Block Text Object Reference=Get Index()
    Font:"Slate Font Info Structure (by ref)"=Get MatchResultValuesFont() -> (MatchResultValuesFont:Slate Font Info Structure)
   ) -> (then)
   > SetFont(self=Get Points(), Font=Get MatchResultValuesFont())
   > SetFont(self=Get Placement(), Font=Get MatchResultValuesFont())
   > SetFont(self=Get Eliminations(), Font=Get MatchResultValuesFont())
   SetText(execute=@n1.then_2, self=Get Index(), Text="LOCTEXT(\"##1\")")
   > SetText(self=Get Placement(), Text="LOCTEXT(\"4\")")
   > SetText(self=Get Eliminations(), Text="LOCTEXT(\"6\")")
   > SetText(self=Get Points(), Text="LOCTEXT(\"280\")")
   RefreshPointsLayout(execute=@n1.then_3) {is=title K2Node_CallFunction}
  ]}
 Set Tracked Stat Times Achieved {nodes [
   n1 = Set Tracked Stat Times Achieved!() -> (then, Item:WBP UIKit Block Text Object Reference, StatString:"String (by ref)") {is=title K2Node_FunctionEntry}
   n2 = Find(
    TargetMap:Map of Strings to Integers=Get TrackedStatsTimesAchieved!(self:Tournament Leaderboard Session Team Entry VM Object Reference=Get LeaderboardSessionTeamEntryVM() -> (LeaderboardSessionTeamEntryVM:Tournament Leaderboard Session Team Entry VM Object Reference)) -> (TrackedStatsTimesAchieved:Map of Strings to Integers)
    Key=@n1.StatString
   ) -> (Value:int)
   > SetText(self=@n1, Text="To Text (Integer)"(Value=SelectInt(A=@n2, B=0, bPickA=@n2.ReturnValue)))
  ]}
 SetEliminations {nodes [
   n1 = SetEliminations!() -> (then, Value:int) {is=title K2Node_FunctionEntry}
   > SetText(self=Get Eliminations(), Text="To Text (Integer)"(Value=@n1))
  ]}
 SetIndex {nodes [
   n1 = SetIndex!() -> (then, Index:int) {is=title K2Node_FunctionEntry}
   > SetText(self=Get Index(), Text=Format Text!(Format="LOCTEXT(\"##{Index}\")", Index:int=@n1) -> (Result))
   > UpdateBackgroundColor!(execute, self:WBP Leaderboard List Entry Base Object Reference, Index:int=int - int(A=Get Index() -> (Index:int=0))) -> (then)
   n2 = Get LeaderboardSessionTeamEntryVM()
   > MakeTextDimmed!(
    execute
    self
    bDim:bool=NOT Boolean(A=Get bIsScoreContributor!(self:Tournament Leaderboard Session Team Entry VM Object Reference=@n2) -> (bIsScoreContributor:bool))
   ) -> (then) {is=title K2Node_CallFunction}
  ]}
 SetPoints {nodes [
   n1 = SetPoints!() -> (then, Value:int) {is=title K2Node_FunctionEntry}
   > SetText(self=Get Points(), Text="To Text (Integer)"(Value=@n1))
  ]}
 SetWins {nodes [
   n1 = SetWins!() -> (then, Value:int) {is=title K2Node_FunctionEntry}
   > SetText(self=Get Placement(), Text="To Text (Integer)"(Value=@n1))
  ]}
]
variables [
 BackgroundMaterial {type=MaterialInstanceDynamic}
 MatchResultValuesFont {
  type=SlateFontInfo
  default {
   bForceMonospaced=false
   bMaterialIsStencil=false
   fontMaterial=/FortUILibrary/UIKit/Buttons/MaterialInstances/SelectableButtons/MI_UIKit_SelectableText_Quiet
   fontObject=&HeadingNow
   letterSpacing=0
   monospacedWidth=1
   size=18
   skewAmount=0
   typefaceFontName=75 Medium
   outlineSettings {
    bApplyOutlineToDropShadows=false
    bMiteredCorners=false
    bSeparateFillAlpha=false
    outlineMaterial=null
    outlineSize=0
    outlineColor=<rgba 0 0 0 1>
   }
  }
 }
]
widgets {
 info {
  inheritedWidgetCount=0
  namedSlotCount=1
  parentClass=/Compete/TournamentLeaderboard/ModalMessages/WBP_LeaderboardListEntry_Base.WBP_LeaderboardListEntry_Base_C
  rootWidgetClass=null
  widgetCount=5
 }
 widgets [
  Main {widgetClassPath=&GridPanel, bIsVariable=true, widget=Main, properties {columnFill [1,1,1,1], rowFill [1]}}
  Index {
   widgetClassPath=/FortUILibrary/Systems/ModularButton/Blocks/WBP_UIKit_Block_Text
   bIsVariable=true
   parent=Main
   widget=Index
   slot {class=GridSlot, properties {horizontalAlignment=HAlign_Left, verticalAlignment=VAlign_Center, padding=<ltrb 32 0 0 0>}}
  }
  Placement {
   widgetClassPath=&WBP_UIKit_Block_Text
   bIsVariable=true
   parent=Main
   widget=Placement
   slot {class=GridSlot, properties {column=1, horizontalAlignment=HAlign_Center, verticalAlignment=VAlign_Center}}
  }
  Eliminations {
   widgetClassPath=&WBP_UIKit_Block_Text
   bIsVariable=true
   parent=Main
   widget=Eliminations
   slot {class=GridSlot, properties {column=2, horizontalAlignment=HAlign_Center, verticalAlignment=VAlign_Center}}
  }
  Points {
   widgetClassPath=&WBP_UIKit_Block_Text
   bIsVariable=true
   parent=Main
   widget=Points
   slot {class=GridSlot, properties {column=3, horizontalAlignment=HAlign_Center, verticalAlignment=VAlign_Center}}
  }
 ]
}
